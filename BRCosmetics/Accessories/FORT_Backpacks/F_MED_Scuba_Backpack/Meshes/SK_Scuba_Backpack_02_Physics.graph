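# Physics asset for the SK_Scuba_Backpack_02 skeletal mesh, a scuba backpack cosmetic accessory.
# It defines two bodies, one on backpack_root and one on dyn_Hose, each a zero-length capsule that is effectively a sphere.
# A single constraint joins the hose to the root with swing and twist all locked, so its angular limits have no effect and the hose cannot rotate relative to the pack through this joint.
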PhysicsAsset /BRCosmetics/Accessories/FORT_Backpacks/F_MED_Scuba_Backpack/Meshes/SK_Scuba_Backpack_02_Physics
bodies {backpack_root [{shapeType=Capsule, length=0, radius=5, rotation=<rot 0 90 -90>}], dyn_Hose [{shapeType=Capsule, length=0, radius=5}]}
constraints [{
  bone1Name=dyn_Hose
  bone2Name=backpack_root
  swing1LimitDegrees=45
  swing1Motion=Locked
  swing2LimitDegrees=45
  swing2Motion=Locked
  twistLimitDegrees=45
  twistMotion=Locked
 }]
